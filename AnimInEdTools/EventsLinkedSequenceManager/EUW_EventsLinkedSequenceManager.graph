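# Editor utility widget that creates and manages anim sequences linked to a level sequence's Control Rig bindings.
# It reads project and shot metadata from the sequence, or parses its path, then lists one row per binding, linked ones first.
# Run clears and rebuilds links, using a validator to decide which sequences to re-export at required frame ranges.
# Export runs embedded Python, copies the project's compression settings and tags new assets with metadata.
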EditorUtilityWidgetBlueprint /AnimInEdTools/EventsLinkedSequenceManager/EUW_EventsLinkedSequenceManager
graphs [
 Anim Dialog {nodes [
   n1 = Anim Dialog!() -> (then, AnimSequence Path:String, LevelSequence Path:String, ExportInWorld:bool=~, isCustomRange:bool=~, StartFrame:int, EndFrame:int) {is=title K2Node_FunctionEntry}
   > PrintString(InString="To String (Boolean)"(InBool=@n1.isCustomRange))
   > Branch(Condition=IsValid!(InBinding:"Movie Scene Binding Proxy Structure (by ref)"=Get Current Binding() -> (Current Binding:Movie Scene Binding Proxy Structure)) -> (ReturnValue))
   > n2 = GetFocusedLevelSequence!(execute) -> (then, ReturnValue:Level Sequence Object Reference)
   n3 = Get Level Sequence() -> (Level Sequence:Level Sequence Object Reference)
   > n4 = Branch(Condition="Equal (Object)"(A=@n2, B=@n3))
   n5 = Create Anim Sequence!(
    execute=[@n9.then, @n4.then]
    self
    Level Sequence:Level Sequence Object Reference=Get Level Sequence()
    New AnimSequence Path:String=Get AnimSequence Path() -> (AnimSequence Path:String)
    Binding:Movie Scene Binding Proxy Structure=Get Current Binding()
    Project:BP Animation Browser Project Object Reference=Get Current Project() -> (Current Project:BP Animation Browser Project Object Reference)
    ExportInWorld:bool=Get ExportInWorld() -> (ExportInWorld:bool)
    isCustomRange:bool=Get isCustomRange() -> (isCustomRange:bool)
    StartFrame:int=Get StartFrame() -> (StartFrame:int=0)
    EndFrame:int=Get EndFrame() -> (EndFrame:int=0)
   ) -> (then, Anim Sequence:Anim Sequence Object Reference)
   > Is Valid(InputObject=@n5)
   > SetMetadataTag!(execute, Object:Object Reference=@n5, Tag:Name=Project:, Value:String=Get ProjectName() -> (ProjectName:String)) -> (then)
   > SetMetadataTag(Object=@n5, Tag=SubFolder:, Value=Get SubFolderName() -> (SubFolderName:String))
   > SetMetadataTag(Object=@n5, Tag=ShotNumber:, Value=Get ShotNumber() -> (ShotNumber:String))
   > SetMetadataTag(Object=@n5, Tag="Exported in World Space: ", Value="To String (Boolean)"(InBool=Get ExportInWorld()))
   > n6 = For Each Loop(Array:Array of EUW Events Linked Sequence Row Object References=Get EUW Rows() -> (EUW Rows:Array of EUW Events Linked Sequence Row Object References)) -> (Array Element:EUW Events Linked Sequence Row Object Reference)
   > n7 = GetName!(
    execute
    InBinding:"Movie Scene Binding Proxy Structure (by ref)"=Get In Binding!(self:EUW Events Linked Sequence Row Object Reference=@n6) -> (In Binding:Movie Scene Binding Proxy Structure)
   ) -> (then, ReturnValue:String)
   > n8 = GetName(InBinding=Get Current Binding())
   > Branch(Condition="Equal Exactly (String)"(A=@n8, B=@n7))
   > Set AnimSequence!(execute, AnimSequence:Anim Sequence Object Reference=@n5, self:EUW Events Linked Sequence Row Object Reference=@n6) -> (then, Output_Get:Anim Sequence Object Reference)
   > Update Appearance(self:EUW Events Linked Sequence Row Object Reference=@n6)
   n9 = OpenEditorForAssets!(
    execute=@n4.else
    self:Asset Editor Subsystem Object Reference=Get AssetEditorSubsystem()
    Assets:Array of Object References=Make Array("[0]":Object Reference=@n3) -> (Array:Array of Object References)
    OpenedMethod:EAssetTypeActivationOpenedMethod Enum=Edit
   ) -> (then, ReturnValue:bool)
  ]}
 Create Anim Dialog {nodes [
   Set Current Binding(Current Binding:Movie Scene Binding Proxy Structure) -> (Output_Get:Movie Scene Binding Proxy Structure)
   > n1 = Construct Object from Class(Class=/AnimInEdTools/AnimationBrowser/EUW_Animation_Dialog.EUW_Animation_Dialog_C) -> (ReturnValue:EUW Animation Dialog Object Reference)
   > Set Initial Settings!(
    execute
    self:EUW Animation Dialog Object Reference=@n1
    Project:BP Animation Browser Project Object Reference
    Selected Creature:Animatable Creature Object Reference
    Developer Mode:bool=false
    Set Plugin:bool=false
    Animatable Creatures:Array of Animatable Creature Object References
    Show Project Dialog:bool=true
    Override Anim Path Validation:bool=false
   ) -> (then) {is=title K2Node_CallFunction}
   > Bind Event to Generate Animation Paths(self:EUW Animation Dialog Object Reference=@n1, Delegate=Create Event()) {is=title K2Node_AddDelegate}
   AddChild(Content=@n1)
   Create Anim Dialog!() -> (then, Binding:Movie Scene Binding Proxy Structure) {is=title K2Node_FunctionEntry}
  ]}
 Create Anim Sequence {nodes [
   n1 = Create Anim Sequence!() -> (
    then
    Level Sequence:Level Sequence Object Reference
    New AnimSequence Path:String
    Binding:Movie Scene Binding Proxy Structure
    Project:BP Animation Browser Project Object Reference
    ExportInWorld:bool=~
    isCustomRange:bool=~
    StartFrame:int
    EndFrame:int
   ) {is=title K2Node_FunctionEntry}
   > PrintString(InString=Append(A="You are trying to create this AnimSeq: ", B=@n1.New AnimSequence Path))
   > n2 = Sequence()
   n3 = Is Valid(exec=@n2.then_1, InputObject=Get Skeleton() -> (Skeleton:Skeleton Object Reference))
   n4 = Get New AnimSequence Path() -> (New AnimSequence Path:String)
   > Execute Python Script!(
    execute
    PythonScript="import unreal\r\n\r\n## Grab the Level Editor World\r\neditor_subsystem = unreal.get_editor_subsystem(unreal.UnrealEditorSubsystem)\r\nworld = editor_subsystem.get_editor_world()\r\n \r\n## Create animation sequence export options\r\nanim_seq_export_options = unreal.AnimSeqExportOption()\r\nanim_seq_export_options.export_transforms = True\r\nanim_seq_export_options.export_morph_targets = True\r\nanim_seq_export_options.record_in_world_space = is_world_space\r\nanim_seq_export_options.use_custom_time_range = custom_range\r\nanim_seq_export_options.custom_start_frame = range_start\r\nanim_seq_export_options.custom_end_frame = range_end\r\nanim_seq_export_options.bake_timecode = True\r\n\r\n \r\n## Get asset tools\r\nasset_tools = unreal.AssetToolsHelpers.get_asset_tools()\r\n\r\n## Create an anim sequence factory, assign it the target skeleton\r\nanim_sequence_factory = unreal.AnimSequenceFactory()\r\nanim_sequence_factory.target_skeleton = skeleton\r\nanim_sequence = unreal.AssetTools.create_asset(asset_tools, asset_name=asset_name, package_path=package_path, asset_class=unreal.AnimSequence, factory=anim_sequence_factory)  \r\n\r\n## Export\r\nunreal.ControlRigSequencerLibrary.export_anim_sequence_from_sequencer(anim_sequence, anim_seq_export_options, binding, True)"
    package_path:String=GetPath!(InPath:String=@n4) -> (ReturnValue:String)
    asset_name:String=GetBaseFilename(InPath=@n4, bRemovePath=true)
    level_sequence:Level Sequence Object Reference=Get Level Sequence()
    binding:Movie Scene Binding Proxy Structure=Get Binding() -> (Binding:Movie Scene Binding Proxy Structure)
    skeleton:Skeleton Object Reference=Get Skeleton()
    is_world_space:bool=Get ExportInWorld()
    custom_range:bool=Get isCustomRange()
    range_start:Frame Number Structure=Make Frame Number!(Value:int=Get StartFrame() -> (StartFrame=0)) -> (FrameNumber:Frame Number Structure) {is=title K2Node_MakeStruct}
    range_end:Frame Number Structure=Make Frame Number(Value=Get EndFrame() -> (EndFrame=0))
   ) -> (then, ReturnValue) {inputs [package_path, asset_name, level_sequence, binding, skeleton, is_world_space, custom_range, range_start, range_end]}
   > n5 = LoadAsset(AssetPath=Get New AnimSequence Path())
   > n6 = Cast To AnimSequence!(execute, Object=@n5) -> (then, CastFailed, AsAnim Sequence:Anim Sequence Object Reference)
   > Set Created AnimSequence(Created AnimSequence:Anim Sequence Object Reference=@n6) -> (Output_Get:Anim Sequence Object Reference)
   > n7 = Sequence()
   n8 = Get Compression Settings!(self:BP Animation Browser Project Object Reference=Get Project() -> (Project:BP Animation Browser Project Object Reference)) -> (Compression Settings:Anim Sequence Soft Object Reference)
   > Branch(Condition=IsValidSoftObjectReference(SoftObjectReference=@n8))
   > n9 = LoadAsset_Blocking(Asset=@n8)
   > n10 = Cast To AnimSequence(Object=@n9)
   > SetBoneCompressionSettings!(
    execute
    AnimationSequence:Anim Sequence Object Reference=Get Created AnimSequence() -> (Created AnimSequence:Anim Sequence Object Reference)
    CompressionSettings:Anim Bone Compression Settings Object Reference=GetBoneCompressionSettings!(AnimationSequence:Anim Sequence Object Reference=@n10) -> (CompressionSettings:Anim Bone Compression Settings Object Reference)
   ) -> (then)
   > SetCurveCompressionSettings!(
    execute
    AnimationSequence:Anim Sequence Object Reference=Get Created AnimSequence()
    CompressionSettings:Anim Curve Compression Settings Object Reference=GetCurveCompressionSettings!(AnimationSequence:Anim Sequence Object Reference=@n10) -> (CompressionSettings:Anim Curve Compression Settings Object Reference)
   ) -> (then)
   > SetAnimationInterpolationType!(
    execute
    AnimationSequence:Anim Sequence Object Reference=Get Created AnimSequence()
    InterpolationType:EAnimInterpolationType Enum=GetAnimationInterpolationType!(AnimationSequence:Anim Sequence Object Reference=@n10) -> (InterpolationType:EAnimInterpolationType Enum=Linear)
   ) -> (then)
   > SaveLoadedAsset(AssetToSave=Get Created AnimSequence(), bOnlyIfIsDirty=false)
   Return Node!(execute=@n7.then_1, Anim Sequence:Anim Sequence Object Reference=Get Created AnimSequence())
   Return Node(execute=@n3.Is Not Valid)
   n11 = Get Current Time!(execute=@n2.then_0, TimeUnit:EMovieSceneTimeUnit Enum=DisplayRate) -> (then, ReturnValue:Movie Scene Sequence Playback Params Structure) {is=title K2Node_CallFunction}
   n12 = Break Movie Scene Sequence Playback Params!(MovieSceneSequencePlaybackParams:"Movie Scene Sequence Playback Params Structure (by ref)"=@n11) -> (
    Frame_FrameNumber_Value:int
    Frame_SubFrame:float=~
    Time:float=~
    MarkedFrame:String
    Timecode:Timecode Structure
    PositionType:EMovieScenePositionType Enum
    UpdateMethod:EUpdatePositionMethod Enum
    bHasJumped:bool=~
   ) {is=title K2Node_BreakStruct}
   > Set Current Time!(
    execute
    PlaybackParams:Movie Scene Sequence Playback Params Structure=Make Movie Scene Sequence Playback Params!(
     Frame_FrameNumber_Value:int=int + int(A=@n12)
     Frame_SubFrame:float=0
     Time:float=0
     MarkedFrame:String
     Timecode:Timecode Structure=(Hours=0,Minutes=0,Seconds=0,Frames=0,Subframe=0,bDropFrameFormat=False)
     PositionType:EMovieScenePositionType Enum=Frame
     UpdateMethod:EUpdatePositionMethod Enum=Play
     bHasJumped:bool=False
    ) -> (MovieSceneSequencePlaybackParams:Movie Scene Sequence Playback Params Structure) {is=title K2Node_MakeStruct}
    TimeUnit:EMovieSceneTimeUnit Enum=DisplayRate
   ) -> (then) {is=title K2Node_CallFunction}
   > Set Current Time(PlaybackParams=@n11, TimeUnit=DisplayRate)
   > n13 = GetBindingID!(
    execute
    Sequence:Movie Scene Sequence Object Reference=Get Level Sequence()
    InBinding:"Movie Scene Binding Proxy Structure (by ref)"=Get Binding()
   ) -> (then, ReturnValue:Movie Scene Object Binding ID Structure)
   > n14 = For Each Loop with Break!(
    Exec:Exec
    Array:Array of Object References=GetBoundObjects!(ObjectBinding:Movie Scene Object Binding ID Structure=@n13) -> (ReturnValue:Array of Object References)
    Break:Exec=@n17.then
   ) -> (LoopBody:Exec, Array Element:Object Reference, Array Index:int, Completed:Exec)
   > n15 = Cast To Actor!(execute, Object=@n14) -> (then, CastFailed, AsActor:Actor Object Reference) {is=title K2Node_DynamicCast}
   n16 = GetComponentByClass!(self:Actor Object Reference=@n15, ComponentClass:Actor Component Class Reference=/Script/Engine.SkeletalMeshComponent) -> (ReturnValue:Skeletal Mesh Component Object Reference)
   > n17 = Set Skeleton(Skeleton:Skeleton Object Reference=Get Skeleton!(self:Skeletal Mesh Object Reference=Get SkeletalMesh!(self:Skinned Mesh Component Object Reference=@n16) -> (SkeletalMesh:Skeletal Mesh Object Reference)) -> (Skeleton)) -> (Output_Get:Skeleton Object Reference)
  ]}
 Create Rows {nodes [
   Create Rows() {is=title K2Node_FunctionEntry}
   > n1 = Sequence()
   > n2 = For Each Loop(Array=Get EUW Rows())
   Clear(execute=@n2.Completed, TargetArray:Array of EUW Events Linked Sequence Row Object References=Get EUW Rows())
   RemoveFromParent(execute=@n2.LoopBody, self=@n2)
   n3 = Get Level Sequence()
   Is Valid(exec=@n1.then_1, InputObject=@n3)
   > n4 = GetBindings!(execute, Sequence:Movie Scene Sequence Object Reference=@n3) -> (then, ReturnValue:Array of Movie Scene Binding Proxy Structures)
   > n5 = GetAnimSequenceLinkFromLevelSequence!(execute, InLevelSequence:Level Sequence Object Reference=Get Level Sequence()) -> (then, ReturnValue:Level Sequence Anim Sequence Link Object Reference)
   > Set Anim Sequence Link(Anim Sequence Link:Level Sequence Anim Sequence Link Object Reference=@n5) -> (Output_Get:Level Sequence Anim Sequence Link Object Reference)
   > n6 = For Each Loop(Array:Array of Movie Scene Binding Proxy Structures=@n4) -> (Array Element:Movie Scene Binding Proxy Structure)
   > Set Local Current Binding(Local Current Binding:Movie Scene Binding Proxy Structure=@n6) -> (Output_Get:Movie Scene Binding Proxy Structure)
   > n7 = Sequence!(execute) -> (then_0, then_1, then_2:Exec)
   Branch(execute=@n7.then_1, Condition=Get Is Current Binding CR() -> (Is Current Binding CR:bool))
   > n8 = Sequence()
   n9 = For Each Loop!(
    Exec=@n8.then_1
    Array:Array of Level Sequence Anim Sequence Link Item Structures=Get AnimSequenceLinks!(self:Level Sequence Anim Sequence Link Object Reference=Get Anim Sequence Link() -> (Anim Sequence Link:Level Sequence Anim Sequence Link Object Reference)) -> (AnimSequenceLinks:Array of Level Sequence Anim Sequence Link Item Structures)
   ) -> (
    LoopBody
    Array Element_SkelTrackGuid:Guid Structure
    Array Element_PathToAnimSequence:Soft Object Path Structure
    Array Element_bAutoBake:bool=~
    Array Element_bExportTransforms:bool=~
    Array Element_bExportMorphTargets:bool=~
    Array Element_bExportAttributeCurves:bool=~
    Array Element_bExportMaterialCurves:bool=~
    Array Element_bSkipCurvesWithZeroValue:bool=~
    Array Element_bSetRetargetSourceAsset:bool=~
    Array Element_Interpolation:EAnimInterpolationType Enum
    Array Element_CurveInterpolation:ERichCurveInterpMode Enum
    Array Element_bRecordInWorldSpace:bool=~
    Array Element_bEvaluateAllSkeletalMeshComponents:bool=~
    Array Element_IncludeAnimationNames:Array of Strings
    Array Element_ExcludeAnimationNames:Array of Strings
    Array Element_bRemoveExcludedCurves:bool=~
    Array Element_WarmUpFrames:Frame Number Structure
    Array Element_DelayBeforeStart:Frame Number Structure
    Array Element_bUseCustomTimeRange:bool=~
    Array Element_CustomStartFrame:Frame Number Structure
    Array Element_CustomEndFrame:Frame Number Structure
    Array Element_CustomDisplayRate:Frame Rate Structure
    Array Element_bUseCustomFrameRate:bool=~
    Array Element_CustomFrameRate:Frame Rate Structure
    Array Index
    Completed
   )
   > n10 = FindBindingById!(execute, Sequence:Movie Scene Sequence Object Reference=Get Level Sequence(), BindingId:Guid Structure=@n9) -> (then, ReturnValue_BindingID:Guid Structure, ReturnValue_Sequence:Movie Scene Sequence Object Reference)
   n11 = Get Local Current Binding() -> (Local Current Binding:Movie Scene Binding Proxy Structure)
   > Branch(Condition="Equal (Guid)"(
     A:"Guid Structure (by ref)"=@n10
     B:"Guid Structure (by ref)"=Break Movie Scene Binding Proxy!(MovieSceneBindingProxy:"Movie Scene Binding Proxy Structure (by ref)"=@n11) -> (BindingID:Guid Structure, Sequence:Movie Scene Sequence Object Reference) {is=title K2Node_BreakStruct}
    ) -> (ReturnValue:bool=~) {is=title K2Node_PromotableOperator})
   > n12 = LoadAsset(AssetPath=BreakSoftObjectPath(InSoftObjectPath=@n9.Array Element_PathToAnimSequence))
   > n13 = Cast To AnimSequence(Object=@n12)
   > n14 = Set AnimSequence(AnimSequence=@n13, self=Get Current Row Widget() -> (Current Row Widget:EUW Events Linked Sequence Row Object Reference))
   > SetCheckedState!(
    execute
    self:Check Box Object Reference=Get LinkCheckBox!(self:EUW Events Linked Sequence Row Object Reference=Get Current Row Widget()) -> (LinkCheckBox:Check Box Object Reference)
    InCheckedState:ECheckBoxState Enum=Checked
   ) -> (then)
   > n15 = GetMetadataTag!(execute, Object:Object Reference=@n14, Tag:Name="Exported in World Space: ") -> (then, ReturnValue:String)
   > Branch(Condition=Contains(SearchIn=@n15, Substring=True))
   n16 = Get Current Row Widget()
   > SetCheckedState(
    self=Get WorldSpaceCheckBox!(self:EUW Events Linked Sequence Row Object Reference=@n16) -> (WorldSpaceCheckBox:Check Box Object Reference)
    InCheckedState=Checked
   )
   > Set ExportWorldSpace!(execute, ExportWorldSpace:bool=true, self:EUW Events Linked Sequence Row Object Reference=@n16) -> (then, Output_Get:bool)
   n17 = Create Widget!(
    execute:Exec=@n8.then_0
    WorldContextObject:Object Reference
    Class:User Widget Class Reference=&/EUW_EventsLinkedSequence_Row.EUW_EventsLinkedSequence_Row_C
    OwningPlayer:Player Controller Object Reference
   ) -> (then:Exec, ReturnValue:EUW Events Linked Sequence Row Object Reference)
   > Set Current Row Widget(Current Row Widget:EUW Events Linked Sequence Row Object Reference=@n17) -> (Output_Get:EUW Events Linked Sequence Row Object Reference)
   > AddChild(self=Get ScrollBox() -> (ScrollBox:Scroll Box Object Reference), Content=Get Current Row Widget())
   > n18 = GetDisplayName!(execute:Exec, InBinding:"Movie Scene Binding Proxy Structure (by ref)"=Get Local Current Binding()) -> (then:Exec, ReturnValue:Text)
   > Set Text(
    Text=@n18
    self=Get BindingName!(self:EUW Events Linked Sequence Row Object Reference=Get Current Row Widget()) -> (BindingName:Text Object Reference)
   )
   > Set In Binding!(
    execute
    In Binding:Movie Scene Binding Proxy Structure=Get Local Current Binding()
    self:EUW Events Linked Sequence Row Object Reference=Get Current Row Widget()
   ) -> (then, Output_Get:Movie Scene Binding Proxy Structure)
   > Add(
    TargetArray:Array of EUW Events Linked Sequence Row Object References=Get EUW Rows()
    NewItem:"EUW Events Linked Sequence Row Object Reference (by ref)"=Get Current Row Widget()
   ) {is=title K2Node_CallArrayFunction}
   > Bind Event to Create New Anim Sequence For Binding(self:EUW Events Linked Sequence Row Object Reference=Get Current Row Widget(), Delegate=Create Event()) {is=title K2Node_AddDelegate}
   Update Appearance(execute=@n7.then_2, self=Get Current Row Widget())
   > Set Is Current Binding CR(Is Current Binding CR:bool=false) -> (Output_Get:bool)
   n19 = FindTracksByExactType!(
    execute=@n7.then_0
    InBinding:"Movie Scene Binding Proxy Structure (by ref)"=Get Local Current Binding()
    TrackType:Movie Scene Track Class Reference=/Script/ControlRig.MovieSceneControlRigParameterTrack
   ) -> (then, ReturnValue:Array of Movie Scene Control Rig Parameter Track Object References)
   > n20 = For Each Loop!(Exec, Array:Array of Movie Scene Control Rig Parameter Track Object References=@n19) -> (LoopBody, Array Element:Movie Scene Control Rig Parameter Track Object Reference, Array Index, Completed)
   > n21 = GetDisplayName!(execute, Track:Movie Scene Track Object Reference=@n20) -> (then, ReturnValue)
   > n22 = Branch(Condition=Contains(SearchIn="To String (Text)"(InText=@n21), Substring=locator))
   n23 = GetSections!(execute=@n22.else, Track:Movie Scene Track Object Reference=@n20) -> (then, ReturnValue:Array of Movie Scene Section Object References)
   > n24 = For Each Loop(Array:Array of Movie Scene Section Object References=@n23) -> (Array Element:Movie Scene Section Object Reference)
   > n25 = Cast To MovieSceneControlRigParameterSection!(execute, Object=@n24) -> (then, CastFailed, AsMovie Scene Control Rig Parameter Section:Movie Scene Control Rig Parameter Section Object Reference) {is=title K2Node_DynamicCast}
   > Set Is Current Binding CR(Is Current Binding CR=IsValid!(Object:Object Reference=@n25) -> (ReturnValue))
  ]}
 EventGraph {nodes [
   n1 = "On Property Changed (LevelSequencePropertyView)"!() -> (OutputDelegate, then, PropertyName:Name=None) {is=title K2Node_ComponentBoundEvent}
   GetMetaData(execute=[@n1.then, @n23.then, @n19.then])
   n2 = Get ProjectName()
   n3 = Get ProjectNames() -> (ProjectNames:Array of Strings)
   > Set Current Project(Current Project:BP Animation Browser Project Object Reference="Get (a copy)"(
     Array:Array of BP Animation Browser Project Object References=Get CurrentProjects() -> (CurrentProjects:Array of BP Animation Browser Project Object References)
     Dimension 1=Find Item!(TargetArray=@n3, ItemToFind:"String (by ref)"=@n2) -> (ReturnValue)
    ) -> (Output:BP Animation Browser Project Object Reference)) -> (Output_Get:BP Animation Browser Project Object Reference)
   > Create Rows() {is=title K2Node_CallFunction}
   > Sort Scroll Box By Checkbox()
   "On Clicked (RunButton)"() {is=title K2Node_ComponentBoundEvent}
   > n4 = Sequence!(execute) -> (then_0, then_1)
   > n5 = Create Widget(Class=/AnimInEdTools/EventsValidator/SubTools/EventValidatorFunctions.EventValidatorFunctions_C) -> (ReturnValue:Event Validator Functions Object Reference)
   > Set Validator(Validator:Event Validator Functions Object Reference=@n5) -> (Output_Get:Event Validator Functions Object Reference)
   > ClearLinkedAnimSequences!(execute, InLevelSequence:Level Sequence Object Reference=Get Level Sequence()) -> (then)
   > n6 = For Each Loop(Array:Array of EUW Events Linked Sequence Row Object References=Get EUW Rows()) -> (Array Element:EUW Events Linked Sequence Row Object Reference)
   n7 = ValidateAnimSeq!(
    self:Event Validator Functions Object Reference=Get Validator() -> (Validator:Event Validator Functions Object Reference)
    Project:BP Animation Browser Project Object Reference=Get Current Project()
    AnimSequence:Anim Sequence Object Reference=Get AnimSequence!(self:EUW Events Linked Sequence Row Object Reference=@n6) -> (AnimSequence:Anim Sequence Object Reference)
   ) -> (IsValid:bool, RowFound:bool, RequiredStartFrame:int=0, RequiredEndFrame:int=0, StatusText:String)
   > n8 = Branch(Condition=@n7)
   n9 = IsChecked(self=Get LinkCheckBox(self=@n6))
   > n10 = Branch(Condition=@n9)
   n11 = Execute Python Script!(
    execute=[@n10.then, @n17.then_1]
    PythonScript="import unreal\r\n## Create animation sequence export options\r\nanim_seq_export_options = unreal.AnimSeqExportOption()\r\nanim_seq_export_options.export_transforms = True\r\nanim_seq_export_options.export_morph_targets = True"
   ) -> (then, ReturnValue, anim_seq_export_options:Anim Seq Export Option Object Reference) {outputs [anim_seq_export_options]}
   n12 = Get AnimSequence(self=@n6)
   > n13 = LinkAnimSequence!(
    execute
    Sequence:Level Sequence Object Reference=Get Level Sequence()
    AnimSequence:Anim Sequence Object Reference=@n12
    ExportOptions:Anim Seq Export Option Object Reference=@n11.anim_seq_export_options
    Binding:"Movie Scene Binding Proxy Structure (by ref)"=Get In Binding(self=@n6)
   ) -> (then, ReturnValue:bool)
   SaveLoadedAsset(execute=[@n13.then, @n15.then], AssetToSave=@n12, bOnlyIfIsDirty=false)
   > RemoveFromParent(self=Get Validator())
   Set AnimSequence(execute=[@n17.then_0, @n10.else, @n16.else], self=@n6)
   > n14 = GeneratePaths!(execute, self, EUW Row:EUW Events Linked Sequence Row Object Reference=@n6) -> (then, AnimSequencePath:String, LevelSequencePath:String)
   > n15 = Anim Dialog!(
    execute
    self
    AnimSequence Path:String=@n14
    LevelSequence Path:String=@n14.LevelSequencePath
    ExportInWorld:bool=IsChecked(self=Get WorldSpaceCheckBox(self=@n6))
    isCustomRange:bool=@n7.RowFound
    StartFrame:int=@n7.RequiredStartFrame
    EndFrame:int=@n7.RequiredEndFrame
   ) -> (then) {is=title K2Node_CallFunction}
   n16 = Branch(execute=@n8.else, Condition=@n9)
   > n17 = Sequence()
   SaveLoadedAsset(execute=@n6.Completed, AssetToSave=Get Level Sequence(), bOnlyIfIsDirty=false)
   Sort Scroll Box By Checkbox(execute=@n4.then_1)
   "On Clicked (CurrentLevelSequenceButton)"() {is=title K2Node_ComponentBoundEvent}
   > n18 = GetFocusedLevelSequence()
   > n19 = Set Level Sequence(Level Sequence:Level Sequence Object Reference=@n18) -> (Output_Get:Level Sequence Object Reference)
   Event PreConstruct()
   > SetObject(self=Get LevelSequencePropertyView() -> (LevelSequencePropertyView:Single Property View Object Reference), NewObject=Self-Reference())
   Event Construct()
   > n20 = Set CurrentProjects(CurrentProjects:Array of BP Animation Browser Project Object References=GetProjects!() -> (Projects:Array of BP Animation Browser Project Object References)) -> (Output_Get:Array of BP Animation Browser Project Object References)
   > n21 = For Each Loop(Array:Array of BP Animation Browser Project Object References=@n20) -> (Array Element:BP Animation Browser Project Object Reference)
   n22 = GetFocusedLevelSequence(execute=@n21.Completed)
   > n23 = Set Level Sequence(Level Sequence=@n22)
   Add(
    execute=@n21.LoopBody
    TargetArray:Array of Strings=Get ProjectNames()
    NewItem:"String (by ref)"=Get Name!(self:BP Animation Browser Project Object Reference=@n21) -> (Name:String)
   )
  ]}
 GeneratePaths {nodes [
   n1 = GeneratePaths!() -> (then, EUW Row:EUW Events Linked Sequence Row Object Reference) {is=title K2Node_FunctionEntry}
   > Set Current Binding(Current Binding=Get In Binding(self=@n1))
   > n2 = Branch(Condition=IsEmpty(InString=Get SubShot() -> (SubShot:String)))
   n3 = Get BindingName(self=Get EUW Row() -> (EUW Row:EUW Events Linked Sequence Row Object Reference))
   n4 = Replace(SourceString="To String (Text)"(InText=Get Text(self:Text Object Reference=@n3)), From=" ", To=_)
   n5 = Get SubFolderName()
   n6 = Get ProjectName()
   n7 = Get ShotNumber()
   n8 = GetProjectAC!() -> (Array Element:Animatable Creature Object Reference)
   > n9 = Set AnimPath(AnimPath:String=Append!(
     A=Get Animation Path!(self:Animatable Creature Object Reference=@n8) -> (Animation Path:String)
     B=@n5
     C:String=/
     D:String=@n5
     E:String=@n7
     F:String=/
     G:String=Append!(A=@n6, B=_, C=@n5, D=@n7, E=_, F=@n4) -> (ReturnValue)
    ) -> (ReturnValue)) -> (Output_Get:String)
   Return Node!(
    execute=[@n9.then, @n17.then]
    AnimSequencePath:String=Get AnimPath() -> (AnimPath:String)
    LevelSequencePath:String=Get Object Path String(Object=Get Level Sequence())
   )
   n10 = Get BindingName(self=Get EUW Row())
   n11 = Replace(SourceString="To String (Text)"(InText=Get Text(self=@n10)), From=" ", To=_)
   n12 = Get ProjectName()
   n13 = Get SubShot()
   n14 = Get SubFolderName()
   n15 = Get ShotNumber()
   n16 = GetProjectAC()
   n17 = Set AnimPath(
    execute=@n2.else
    AnimPath=Append!(
     A=Get Animation Path(self=@n16)
     B=@n14
     C=/
     D=@n14
     E=@n15
     F=/
     G:String=Append!(A=@n12, B=_, C=@n14, D=@n15, E=_, F=@n13, G=_, H:String=@n11) -> (ReturnValue)
    ) -> (ReturnValue)
   )
  ]}
 Get Skeletal Bindings {nodes [
   Get Skeletal Bindings!() -> (then, Bindings:Array of Movie Scene Binding Proxy Structures) {is=title K2Node_FunctionEntry}
   > n1 = Sequence!(execute) -> (then_0, then_1, then_2:Exec, then_3:Exec, then_4:Exec)
   Return Node!(
    execute=@n1.then_4
    Skeletal Bindings:Array of Movie Scene Binding Proxy Structures=Get Filtered Skeletal Bindings() -> (Filtered Skeletal Bindings:Array of Movie Scene Binding Proxy Structures)
   )
   > Clear(TargetArray:Array of Movie Scene Binding Proxy Structures=Get Filtered Skeletal Bindings())
   n2 = For Each Loop(
    Exec=@n1.then_1
    Array:Array of Movie Scene Binding Proxy Structures=Get Bindings() -> (Bindings:Array of Movie Scene Binding Proxy Structures)
   ) -> (Array Element:Movie Scene Binding Proxy Structure)
   > n3 = FindTracksByType!(
    execute
    InBinding:"Movie Scene Binding Proxy Structure (by ref)"=@n2
    TrackType:Movie Scene Track Class Reference=/Script/MovieSceneTracks.MovieSceneSkeletalAnimationTrack
   ) -> (then, ReturnValue:Array of Movie Scene Skeletal Animation Track Object References)
   > n4 = For Each Loop(Array:Array of Movie Scene Skeletal Animation Track Object References=@n3) -> (Array Element:Movie Scene Skeletal Animation Track Object Reference)
   > Branch(Condition=IsValid(Object=@n4))
   > Add(
    TargetArray:Array of Movie Scene Binding Proxy Structures=Get Filtered Skeletal Bindings()
    NewItem:"Movie Scene Binding Proxy Structure (by ref)"=@n2
   )
   FindTracksByExactType(InBinding=@n2, TrackType=&MovieSceneSkeletalAnimationTrack) -> (ReturnValue:Array of Movie Scene Skeletal Animation Track Object References)
  ]}
 GetMetaData {nodes [
   GetMetaData() {is=title K2Node_FunctionEntry}
   > n1 = Sequence()
   n2 = Get Level Sequence()
   > n3 = GetMetadataTag(Object=@n2, Tag=Project:)
   > Set ProjectName(ProjectName:String=@n3) -> (Output_Get:String)
   n4 = GetMetadataTag(execute=@n1.then_1, Object=@n2, Tag=SubFolder:)
   > Set SubFolderName(SubFolderName:String=@n4) -> (Output_Get:String)
   n5 = GetMetadataTag(execute=@n1.then_2, Object=@n2, Tag=ShotNumber:)
   > Set ShotNumber(ShotNumber:String=@n5) -> (Output_Get:String)
   n6 = GetMetadataTag(execute=@n1.then_3, Object=@n2, Tag=SubShot:)
   > Set SubShot(SubShot:String=@n6) -> (Output_Get:String)
   n7 = Get ProjectName()
   n8 = Branch(execute=@n1.then_4, Condition="integer > integer"(A=Len(S=@n7)))
   n9 = For Each Loop(Exec=@n8.else, Array:Array of Strings=Get ProjectNames()) -> (Array Element:String)
   n10 = Get Level Sequence()
   n11 = Get Object Path String(Object=@n10)
   n12 = RightChop!(SourceString:String=@n11, Count:int=int + int(A=FindSubstring(SearchIn=@n11, Substring=sequence), B=9)) -> (ReturnValue:String)
   Set SubFolderName(execute=@n9.Completed, SubFolderName=Left!(SourceString:String=@n12, Count:int=3) -> (ReturnValue:String))
   > Set ShotNumber(ShotNumber=RightChop(SourceString=Left(SourceString=@n12, Count=11), Count=7))
   n13 = GetDisplayName!(Object:Object Reference=@n10) -> (ReturnValue:String)
   n14 = Replace(SourceString=Replace(SourceString=@n12, From=Append!(A=@n13, B=., C=@n13) -> (ReturnValue)), From=Get SubFolderName())
   n15 = Get ShotNumber()
   > Set SubShot(SubShot=Replace(SourceString=Replace(SourceString=@n14, From=@n15), From=/))
   Branch(execute=@n9.LoopBody, Condition=Contains(SearchIn=@n11, Substring=@n9))
   > Set ProjectName(ProjectName=@n9)
  ]}
 GetProjectAC {nodes [
   GetProjectAC() {is=title K2Node_FunctionEntry}
   > n1 = For Each Loop(Array:Array of Animatable Creature Object References=Get Animatable Creatures!(Project:BP Animation Browser Project Object Reference=Get Current Project(), bSearchSubClasses:bool=false, __WorldContext:Object Reference) -> (Animatable Creatures:Array of Animatable Creature Object References) {is=title K2Node_CallFunction}) -> (Array Element:Animatable Creature Object Reference)
   > Return Node!(execute, Array Element:Animatable Creature Object Reference=@n1)
  ]}
 GetProjects {nodes [
   GetProjects() {is=title K2Node_FunctionEntry}
   > Clear(TargetArray:Array of BP Animation Browser Project Object References=Get Local Projects() -> (Local Projects:Array of BP Animation Browser Project Object References))
   > n1 = GetAssetsByClass!(
    execute
    self:Asset Registry Interface=GetAssetRegistry()
    ClassPathName:Top Level Asset Path Structure=GetSoftClassTopLevelAssetPath!(SoftClassReference:Object Soft Class Reference=/AnimInEdTools/AnimationBrowser/ProjectSettings/BP_AnimationBrowser_Project.BP_AnimationBrowser_Project_C) -> (ReturnValue:Top Level Asset Path Structure)
    bSearchSubClasses:bool=false
   ) -> (then, OutAssetData:Array of Asset Data Structures, ReturnValue:bool)
   > n2 = For Each Loop(Array:Array of Asset Data Structures=@n1) -> (Array Element:Asset Data Structure)
   Return Node!(execute=@n2.Completed, Projects:Array of BP Animation Browser Project Object References=Get Local Projects())
   n3 = Break Asset Data(AssetData=@n2)
   > n4 = LoadAsset(AssetPath=GetBaseFilename(InPath="To String (Name)"(InName=@n3), bRemovePath=false))
   > n5 = Cast To BP_AnimationBrowser_Project!(execute, Object=@n4) -> (then, CastFailed, AsBP Animation Browser Project:BP Animation Browser Project Object Reference) {is=title K2Node_DynamicCast}
   > Add(
    TargetArray:Array of BP Animation Browser Project Object References=Get Local Projects()
    NewItem:"BP Animation Browser Project Object Reference (by ref)"=@n5
   )
  ]}
 SetProject {nodes [
   SetProject() {is=title K2Node_FunctionEntry}
   > n1 = GetMetadataTag(Object=Get Level Sequence(), Tag=Project:)
   n2 = Get ProjectNames()
   > Set Current Project(Current Project="Get (a copy)"(Array=Get CurrentProjects(), Dimension 1=Find Item(TargetArray=@n2, ItemToFind=@n1)))
  ]}
 Sort Scroll Box By Checkbox {nodes [
   Sort Scroll Box By Checkbox() {is=title K2Node_FunctionEntry}
   > Clear(TargetArray:Array of EUW Events Linked Sequence Row Object References=Get Items To Sort() -> (Items To Sort:Array of EUW Events Linked Sequence Row Object References))
   n1 = GetAllChildren(self=Get ScrollBox())
   > n2 = For Each Loop(Array:Array of Widget Object References=@n1) -> (Array Element:Widget Object Reference)
   n3 = For Each Loop(Exec=@n2.Completed, Array=@n1)
   > n4 = Cast To EUW_EventsLinkedSequence_Row!(execute, Object=@n3) -> (then, CastFailed, AsEUW Events Linked Sequence Row:EUW Events Linked Sequence Row Object Reference) {is=title K2Node_DynamicCast}
   n5 = Get LinkCheckBox(self=@n4)
   > Branch(Condition="Not Equal (Enum)"!(
     A:ECheckBoxState Enum=GetCheckedState!(self:Check Box Object Reference=@n5) -> (ReturnValue:ECheckBoxState Enum=Unchecked)
     B:ECheckBoxState Enum=Checked
    ) -> (ReturnValue:bool))
   > Add(
    TargetArray:Array of EUW Events Linked Sequence Row Object References=Get Items To Sort()
    NewItem:"EUW Events Linked Sequence Row Object Reference (by ref)"=@n4
   )
   ClearChildren(execute=@n3.Completed, self=Get ScrollBox())
   > n6 = For Each Loop(Array:Array of EUW Events Linked Sequence Row Object References=Get Items To Sort()) -> (Array Element:EUW Events Linked Sequence Row Object Reference)
   > AddChild(self=Get ScrollBox(), Content=@n6)
   n7 = Cast To EUW_EventsLinkedSequence_Row(execute=@n2.LoopBody, Object=@n2)
   n8 = Get LinkCheckBox(self=@n7)
   > Branch(Condition="Not Equal (Enum)"(A=GetCheckedState(self=@n8) -> (ReturnValue=Unchecked), B=Unchecked))
   > Add(TargetArray=Get Items To Sort(), NewItem=@n7)
  ]}
 UpdateComboBox {nodes [
   n1 = UpdateComboBox!() -> (then, ComboBox:"Searchable Combo Box (String) Object Reference", Items:Array of Strings, DefaultItem:String, AddBlank:bool=~) {is=title K2Node_FunctionEntry}
   > n2 = Set Current Index(Current Index:int=GetSelectedIndex!(self:"Searchable Combo Box (String) Object Reference"=@n1) -> (ReturnValue:int=0)) -> (Output_Get:int=0)
   > Set Selected Item(Selected Item:String=GetOptionAtIndex!(
     self:"Searchable Combo Box (String) Object Reference"=Get ComboBox() -> (ComboBox:"Searchable Combo Box (String) Object Reference")
     Index:int=@n2
    ) -> (ReturnValue:String)) -> (Output_Get:String)
   n3 = Get ComboBox()
   > n4 = For Loop(LastIndex=int - int(A=GetOptionCount!(self:"Searchable Combo Box (String) Object Reference"=@n3) -> (ReturnValue:int=0)))
   n5 = Branch(
    execute=@n4.Completed
    Condition=Identical!(ArrayA:Array of Strings=Get Items() -> (Items:Array of Strings), ArrayB:Array of Strings=Get Old Items() -> (Old Items:Array of Strings)) -> (ReturnValue:bool) {is=title K2Node_CallArrayFunction}
   )
   ClearOptions(execute=@n5.else, self:"Searchable Combo Box (String) Object Reference"=Get ComboBox())
   > n6 = Sequence!(execute) -> (then_0, then_1, then_2, then_3)
   > Branch(Condition=Get AddBlank() -> (AddBlank:bool))
   > AddOption!(execute, self:"Searchable Combo Box (String) Object Reference"=Get ComboBox(), Option:String) -> (then)
   n7 = Get DefaultItem() -> (DefaultItem:String)
   n8 = Branch(execute=@n6.then_1, Condition=IsEmpty(InString=@n7))
   AddOption(execute=@n8.else, self=Get ComboBox(), Option=@n7)
   n9 = For Each Loop(Exec=@n6.then_2, Array:Array of Strings=Get Items()) -> (Array Element:String)
   > AddOption(self=Get ComboBox(), Option=@n9)
   n10 = Get ComboBox()
   n11 = Get Current Index() -> (Current Index:int=0)
   n12 = Branch(
    execute=@n6.then_3
    Condition="Equal Exactly (String)"(A=GetOptionAtIndex(self=@n10, Index=@n11), B=Get Selected Item() -> (Selected Item:String))
   )
   n13 = Get ComboBox()
   > AddOption(self=@n13, Option=Temp)
   n14 = Get Current Index() -> (Current Index=0)
   > SetSelectedIndex!(execute, self:"Searchable Combo Box (String) Object Reference"=@n13, Index:int=int + int(A=@n14)) -> (then)
   > SetSelectedIndex(self=@n13, Index=@n14)
   > RemoveOption!(execute, self:"Searchable Combo Box (String) Object Reference"=@n13, Option:String=Temp) -> (then, ReturnValue:bool)
   ClearSelection(execute=@n12.else, self:"Searchable Combo Box (String) Object Reference"=Get ComboBox())
   n15 = GetOptionAtIndex(self=Get ComboBox(), Index=@n4)
   n16 = Branch(execute=@n4.LoopBody, Condition="Equal Exactly (String)"(A=@n15, B=Get DefaultItem()))
   Add(execute=@n16.else, TargetArray:Array of Strings=Get Old Items(), NewItem:"String (by ref)"=@n15)
  ]}
]
variables [
 Current Binding {type=MovieSceneBindingProxy, default {bindingId=#1, sequence=null}}
 Current Project {type=BP_AnimationBrowser_Project_C, default=null}
 CurrentProjects {type="array<BP_AnimationBrowser_Project_C>", default []}
 EUW Rows {type="array<EUW_EventsLinkedSequence_Row_C>", default []}
 Level Sequence {type=LevelSequence, default=null}
 ProjectName {type=string, default=""}
 ProjectNames {type="array<string>", default []}
 ShotNumber {type=string, default=""}
 SubFolderName {type=string, default=""}
 SubShot {type=string, default=""}
 Validator {type=EventValidatorFunctions_C, default=null}
]
widgets {
 info {rootWidgetClass=&CanvasPanel, widgetCount=9}
 widgets [
  CanvasPanel_0 {widgetClassPath=&CanvasPanel, widget=CanvasPanel_0}
  VerticalBox_0 {
   widgetClassPath=&VerticalBox
   parent=CanvasPanel_0
   widget=VerticalBox_0
   slot {class=CanvasPanelSlot, properties {layoutData {alignment=<0 0>, anchors {maximum=<1 1>, minimum=<0 0>}, offsets=<ltrb 0 0 0 0>}}}
  }
  HorizontalBox_0 {widgetClassPath=&HorizontalBox, parent=VerticalBox_0, widget=HorizontalBox_0, slot {class=VerticalBoxSlot}}
  LevelSequencePropertyView {
   widgetClassPath=&SinglePropertyView
   parent=HorizontalBox_0
   widget=LevelSequencePropertyView
   properties {propertyName=Level Sequence}
   slot {class=HorizontalBoxSlot, properties {size {sizeRule=Fill, value=2.5}}}
  }
  CurrentLevelSequenceButton {
   widgetClassPath=&EditorUtilityButton
   parent=HorizontalBox_0
   widget=CurrentLevelSequenceButton
   slot {class=HorizontalBoxSlot}
  }
  TextBlock_106 {
   widgetClassPath=&TextBlock
   parent=CurrentLevelSequenceButton
   widget=TextBlock_106
   properties {
    text=Get Current LS
    font {
     bForceMonospaced=false
     bMaterialIsStencil=false
     fontMaterial=null
     fontObject=&Roboto
     letterSpacing=0
     monospacedWidth=1
     size=13.5
     skewAmount=0
     typefaceFontName=Bold
     outlineSettings {
      bApplyOutlineToDropShadows=false
      bMiteredCorners=false
      bSeparateFillAlpha=false
      outlineMaterial=null
      outlineSize=0
      outlineColor=<rgba 0 0 0 1>
     }
    }
   }
   slot {class=ButtonSlot}
  }
  ScrollBox {
   widgetClassPath=&ScrollBox
   bIsVariable=true
   parent=VerticalBox_0
   widget=ScrollBox
   properties {
    consumeMouseWheel=Always
    widgetBarStyle {
     thickness=8
     draggedThumbImage {
      bIsDynamicallyLoaded=false
      drawAs=RoundedBox
      imageType=NoImage
      mirroring=NoMirror
      resourceName=None
      resourceObject=null
      tiling=NoTile
      imageSize=<0 0>
      margin=<ltrb 0 0 0 0>
      outlineSettings {
       bUseBrushTransparency=false
       roundingType=FixedRadius
       width=0
       color {colorUseRule=UseColor_Specified, specifiedColor=<rgba 0 0 0 0>}
       cornerRadii=<4 4 4 4>
      }
      tintColor {colorUseRule=UseColor_Specified, specifiedColor=<rgba 0.2159 0.2159 0.2159 1>}
      uVRegion {bIsValid=false, max=<0 0>, min=<0 0>}
     }
     horizontalBackgroundImage {
      bIsDynamicallyLoaded=false
      drawAs=NoDrawType
      imageType=NoImage
      mirroring=NoMirror
      resourceName=None
      resourceObject=null
      tiling=NoTile
      imageSize=<0 0>
      margin=<ltrb 0 0 0 0>
      outlineSettings {
       bUseBrushTransparency=false
       roundingType=HalfHeightRadius
       width=0
       color {colorUseRule=UseColor_Specified, specifiedColor=<rgba 0 0 0 0>}
       cornerRadii=<0 0 0 0>
      }
      tintColor {colorUseRule=UseColor_Specified, specifiedColor=<rgba 1 1 1 1>}
      uVRegion {bIsValid=false, max=<0 0>, min=<0 0>}
     }
     horizontalBottomSlotImage {
      bIsDynamicallyLoaded=false
      drawAs=NoDrawType
      imageType=NoImage
      mirroring=NoMirror
      resourceName=None
      resourceObject=null
      tiling=NoTile
      imageSize=<0 0>
      margin=<ltrb 0 0 0 0>
      outlineSettings {
       bUseBrushTransparency=false
       roundingType=HalfHeightRadius
       width=0
       color {colorUseRule=UseColor_Specified, specifiedColor=<rgba 0 0 0 0>}
       cornerRadii=<0 0 0 0>
      }
      tintColor {colorUseRule=UseColor_Specified, specifiedColor=<rgba 1 1 1 1>}
      uVRegion {bIsValid=false, max=<0 0>, min=<0 0>}
     }
     horizontalTopSlotImage {
      bIsDynamicallyLoaded=false
      drawAs=NoDrawType
      imageType=NoImage
      mirroring=NoMirror
      resourceName=None
      resourceObject=null
      tiling=NoTile
      imageSize=<0 0>
      margin=<ltrb 0 0 0 0>
      outlineSettings {
       bUseBrushTransparency=false
       roundingType=HalfHeightRadius
       width=0
       color {colorUseRule=UseColor_Specified, specifiedColor=<rgba 0 0 0 0>}
       cornerRadii=<0 0 0 0>
      }
      tintColor {colorUseRule=UseColor_Specified, specifiedColor=<rgba 1 1 1 1>}
      uVRegion {bIsValid=false, max=<0 0>, min=<0 0>}
     }
     hoveredThumbImage {
      bIsDynamicallyLoaded=false
      drawAs=RoundedBox
      imageType=NoImage
      mirroring=NoMirror
      resourceName=None
      resourceObject=null
      tiling=NoTile
      imageSize=<0 0>
      margin=<ltrb 0 0 0 0>
      outlineSettings {
       bUseBrushTransparency=false
       roundingType=FixedRadius
       width=0
       color {colorUseRule=UseColor_Specified, specifiedColor=<rgba 0 0 0 0>}
       cornerRadii=<4 4 4 4>
      }
      tintColor {colorUseRule=UseColor_Specified, specifiedColor=<rgba 0.2159 0.2159 0.2159 1>}
      uVRegion {bIsValid=false, max=<0 0>, min=<0 0>}
     }
     normalThumbImage {
      bIsDynamicallyLoaded=false
      drawAs=RoundedBox
      imageType=NoImage
      mirroring=NoMirror
      resourceName=None
      resourceObject=null
      tiling=NoTile
      imageSize=<0 0>
      margin=<ltrb 0 0 0 0>
      outlineSettings {
       bUseBrushTransparency=false
       roundingType=FixedRadius
       width=0
       color {colorUseRule=UseColor_Specified, specifiedColor=<rgba 0 0 0 0>}
       cornerRadii=<4 4 4 4>
      }
      tintColor {colorUseRule=UseColor_Specified, specifiedColor=<rgba 0.0953 0.0953 0.0953 1>}
      uVRegion {bIsValid=false, max=<0 0>, min=<0 0>}
     }
     verticalBackgroundImage {
      bIsDynamicallyLoaded=false
      drawAs=NoDrawType
      imageType=NoImage
      mirroring=NoMirror
      resourceName=None
      resourceObject=null
      tiling=NoTile
      imageSize=<0 0>
      margin=<ltrb 0 0 0 0>
      outlineSettings {
       bUseBrushTransparency=false
       roundingType=HalfHeightRadius
       width=0
       color {colorUseRule=UseColor_Specified, specifiedColor=<rgba 0 0 0 0>}
       cornerRadii=<0 0 0 0>
      }
      tintColor {colorUseRule=UseColor_Specified, specifiedColor=<rgba 1 1 1 1>}
      uVRegion {bIsValid=false, max=<0 0>, min=<0 0>}
     }
     verticalBottomSlotImage {
      bIsDynamicallyLoaded=false
      drawAs=NoDrawType
      imageType=NoImage
      mirroring=NoMirror
      resourceName=None
      resourceObject=null
      tiling=NoTile
      imageSize=<0 0>
      margin=<ltrb 0 0 0 0>
      outlineSettings {
       bUseBrushTransparency=false
       roundingType=HalfHeightRadius
       width=0
       color {colorUseRule=UseColor_Specified, specifiedColor=<rgba 0 0 0 0>}
       cornerRadii=<0 0 0 0>
      }
      tintColor {colorUseRule=UseColor_Specified, specifiedColor=<rgba 1 1 1 1>}
      uVRegion {bIsValid=false, max=<0 0>, min=<0 0>}
     }
     verticalTopSlotImage {
      bIsDynamicallyLoaded=false
      drawAs=NoDrawType
      imageType=NoImage
      mirroring=NoMirror
      resourceName=None
      resourceObject=null
      tiling=NoTile
      imageSize=<0 0>
      margin=<ltrb 0 0 0 0>
      outlineSettings {
       bUseBrushTransparency=false
       roundingType=HalfHeightRadius
       width=0
       color {colorUseRule=UseColor_Specified, specifiedColor=<rgba 0 0 0 0>}
       cornerRadii=<0 0 0 0>
      }
      tintColor {colorUseRule=UseColor_Specified, specifiedColor=<rgba 1 1 1 1>}
      uVRegion {bIsValid=false, max=<0 0>, min=<0 0>}
     }
    }
    widgetStyle {
     barThickness=8
     bottomShadowBrush {
      bIsDynamicallyLoaded=false
      drawAs=Box
      imageType=FullColor
      mirroring=NoMirror
      resourceName=../../../Engine/Content/Slate/Common/ScrollBoxShadowBottom.png
      resourceObject=null
      tiling=NoTile
      imageSize=<16 8>
      margin=<ltrb 0.5 0 0.5 1>
      outlineSettings {
       bUseBrushTransparency=false
       roundingType=HalfHeightRadius
       width=0
       color {colorUseRule=UseColor_Specified, specifiedColor=<rgba 0 0 0 0>}
       cornerRadii=<0 0 0 0>
      }
      tintColor {colorUseRule=UseColor_Specified, specifiedColor=<rgba 1 1 1 1>}
      uVRegion {bIsValid=false, max=<0 0>, min=<0 0>}
     }
     horizontalScrolledContentPadding=<ltrb 0 0 1 0>
     leftShadowBrush {
      bIsDynamicallyLoaded=false
      drawAs=Box
      imageType=FullColor
      mirroring=NoMirror
      resourceName=../../../Engine/Content/Slate/Common/ScrollBoxShadowLeft.png
      resourceObject=null
      tiling=NoTile
      imageSize=<8 16>
      margin=<ltrb 1 0.5 0 0.5>
      outlineSettings {
       bUseBrushTransparency=false
       roundingType=HalfHeightRadius
       width=0
       color {colorUseRule=UseColor_Specified, specifiedColor=<rgba 0 0 0 0>}
       cornerRadii=<0 0 0 0>
      }
      tintColor {colorUseRule=UseColor_Specified, specifiedColor=<rgba 1 1 1 1>}
      uVRegion {bIsValid=false, max=<0 0>, min=<0 0>}
     }
     rightShadowBrush {
      bIsDynamicallyLoaded=false
      drawAs=Box
      imageType=FullColor
      mirroring=NoMirror
      resourceName=../../../Engine/Content/Slate/Common/ScrollBoxShadowRight.png
      resourceObject=null
      tiling=NoTile
      imageSize=<8 16>
      margin=<ltrb 0 0.5 1 0.5>
      outlineSettings {
       bUseBrushTransparency=false
       roundingType=HalfHeightRadius
       width=0
       color {colorUseRule=UseColor_Specified, specifiedColor=<rgba 0 0 0 0>}
       cornerRadii=<0 0 0 0>
      }
      tintColor {colorUseRule=UseColor_Specified, specifiedColor=<rgba 1 1 1 1>}
      uVRegion {bIsValid=false, max=<0 0>, min=<0 0>}
     }
     topShadowBrush {
      bIsDynamicallyLoaded=false
      drawAs=Box
      imageType=FullColor
      mirroring=NoMirror
      resourceName=../../../Engine/Content/Slate/Common/ScrollBoxShadowTop.png
      resourceObject=null
      tiling=NoTile
      imageSize=<16 8>
      margin=<ltrb 0.5 1 0.5 0>
      outlineSettings {
       bUseBrushTransparency=false
       roundingType=HalfHeightRadius
       width=0
       color {colorUseRule=UseColor_Specified, specifiedColor=<rgba 0 0 0 0>}
       cornerRadii=<0 0 0 0>
      }
      tintColor {colorUseRule=UseColor_Specified, specifiedColor=<rgba 1 1 1 1>}
      uVRegion {bIsValid=false, max=<0 0>, min=<0 0>}
     }
     verticalScrolledContentPadding=<ltrb 0 0 0 1>
    }
   }
   slot {class=VerticalBoxSlot, properties {size {sizeRule=Fill, value=1}}}
  }
  RunButton {
   widgetClassPath=&Button
   parent=VerticalBox_0
   widget=RunButton
   properties {widgetStyle {
     clickedSlateSound {resourceObject=null}
     disabled {
      bIsDynamicallyLoaded=false
      drawAs=RoundedBox
      imageType=NoImage
      mirroring=NoMirror
      resourceName=None
      resourceObject=null
      tiling=NoTile
      imageSize=<0 0>
      margin=<ltrb 0 0 0 0>
      outlineSettings {
       bUseBrushTransparency=false
       roundingType=FixedRadius
       width=1
       color {colorUseRule=UseColor_ColorTable, specifiedColor=<rgba 0 0 0 0>}
       cornerRadii=<4 4 4 4>
      }
      tintColor {colorUseRule=UseColor_Specified, specifiedColor=<rgba 0.0395 0.0395 0.0395 1>}
      uVRegion {bIsValid=false, max=<0 0>, min=<0 0>}
     }
     disabledForeground {colorUseRule=UseColor_ColorTable, specifiedColor=<rgba 0 0 0 0>}
     hovered {
      bIsDynamicallyLoaded=false
      drawAs=RoundedBox
      imageType=NoImage
      mirroring=NoMirror
      resourceName=None
      resourceObject=null
      tiling=NoTile
      imageSize=<32 32>
      margin=<ltrb 0 0 0 0>
      outlineSettings {
       bUseBrushTransparency=false
       roundingType=FixedRadius
       width=1
       color {colorUseRule=UseColor_ColorTable, specifiedColor=<rgba 0 0 0 0>}
       cornerRadii=<4 4 4 4>
      }
      tintColor {colorUseRule=UseColor_Specified, specifiedColor=<rgba 0.0953 0.0953 0.0953 1>}
      uVRegion {bIsValid=false, max=<0 0>, min=<0 0>}
     }
     hoveredForeground {colorUseRule=UseColor_ColorTable, specifiedColor=<rgba 0 0 0 0>}
     hoveredSlateSound {resourceObject=null}
     normal {
      bIsDynamicallyLoaded=false
      drawAs=RoundedBox
      imageType=NoImage
      mirroring=NoMirror
      resourceName=None
      resourceObject=null
      tiling=NoTile
      imageSize=<32 32>
      margin=<ltrb 0 0 0 0>
      outlineSettings {
       bUseBrushTransparency=false
       roundingType=FixedRadius
       width=1
       color {colorUseRule=UseColor_ColorTable, specifiedColor=<rgba 0 0 0 0>}
       cornerRadii=<4 4 4 4>
      }
      tintColor {colorUseRule=UseColor_Specified, specifiedColor=<rgba 0.0395 0.0395 0.0395 1>}
      uVRegion {bIsValid=false, max=<0 0>, min=<0 0>}
     }
     normalForeground {colorUseRule=UseColor_ColorTable, specifiedColor=<rgba 0 0 0 0>}
     normalPadding=<ltrb 12 1.5 12 1.5>
     pressed {
      bIsDynamicallyLoaded=false
      drawAs=RoundedBox
      imageType=NoImage
      mirroring=NoMirror
      resourceName=None
      resourceObject=null
      tiling=NoTile
      imageSize=<32 32>
      margin=<ltrb 0 0 0 0>
      outlineSettings {
       bUseBrushTransparency=false
       roundingType=FixedRadius
       width=1
       color {colorUseRule=UseColor_ColorTable, specifiedColor=<rgba 0 0 0 0>}
       cornerRadii=<4 4 4 4>
      }
      tintColor {colorUseRule=UseColor_Specified, specifiedColor=<rgba 0.0284 0.0284 0.0284 1>}
      uVRegion {bIsValid=false, max=<0 0>, min=<0 0>}
     }
     pressedForeground {colorUseRule=UseColor_ColorTable, specifiedColor=<rgba 0 0 0 0>}
     pressedPadding=<ltrb 12 2.5 12 0.5>
     pressedSlateSound {resourceObject=null}
    }}
   slot {class=VerticalBoxSlot}
  }
  TextBlock_4 {widgetClassPath=&TextBlock, parent=RunButton, widget=TextBlock_4, properties {text=Export / Link}, slot {class=ButtonSlot}}
 ]
}
